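# Translucent unlit Niagara sprite material that draws audio-reactive equalizer bars for the BoomBox cosmetic aura.
# An EQ function uses the Sparks audio FFT texture and beat parameters, with a noise-driven fake FFT selected by the "Is in Front End" dynamic parameter.
# Vertically mirrored UVs, stepped box masks, a three-color gradient and edge glow shape the look, with time-of-day brightness.
# A custom atan2 rotation builds camera-facing billboard offset and normals.
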Material /SparksCosmetics/Items/Auras/Assets/NS_Aura_BoomBox/M_Aura_BoomBox_EQBars
graph {
 properties {bUsedWithNiagaraSprites=true, blendMode=BLEND_Translucent, pixelDepthOffsetMode=PDOM_Legacy, shadingModel=MSM_Unlit, usedShadingModels=Unlit}
 nodes [
  Add(A=TextureCoordinate(), B=Constant2Vector() {r=0.45})
  n1 = Constant3Vector() {constant=<rgba 1 0.8788 0.7292 1>}
  n2 = Constant3Vector() {constant=<rgba 1 0.8001 0.5833 1>}
  n3 = Constant() {r=1}
  n4 = Constant3Vector() {constant=<rgba 1 0.6051 0 1>}
  n5 = Constant3Vector() {constant=<rgba 2 0.0316 0 1>}
  n6 = MaterialFunctionCall(Float2=NamedRerouteUsage()) {materialFunction=&BreakOutFloat2Components}
  n7 = Saturate(None=Multiply(A=Power(Base=@n6.G) {constExponent=1}) {constB=4})
  n8 = Multiply(A=MaterialFunctionCall!(
    Color 1=AppendVector(A=@n2, B=@n3)
    Color 2=AppendVector(A=@n4, B=@n3)
    Color 3=AppendVector(A=@n5, B=@n3)
    middle=Constant() {r=0.25}
    t=MaterialFunctionCall(X=@n7, Y=@n6.R) {materialFunction=&MakeFloat2}.Result
   ) {materialFunction=/Game/Effects/Niagara/MasterMaterials/Functions/Mojo_MF/MF_FX_Lerp3_Color}.Result V3)
  Add(A=@n1, B=@n8)
  n9 = Constant() {r=0.15}
  n10 = Constant() {desc=Num Bar Width, r=15}
  n11 = StaticBool() {value=true}
  n12 = StaticBool()
  n13 = Constant() {r=0.1}
  n14 = Constant() {r=0.01}
  n15 = Constant() {desc=Num Bar Vert, r=24}
  n16 = CollectionParameter() {collection=/FMAudioReactiveData/AudioAnalysis/MPC_Sparks_Audio_Parameters, parameterName=MeasureBeat}
  n17 = Saturate(None=NamedRerouteUsage())
  n18 = MaterialFunctionCall!(
   UV
   Square Grid Width=@n9
   RandomizeUVPlacement=@n11
   RadialUV=@n12
   BarAmount Vert=@n15
   BarAmount=@n10
   BarWidth=@n13
   MountainHardness=@n14
   FFT=NamedRerouteUsage()
   Use Fake FFT Noise=NamedRerouteDeclaration(None=StaticBool()) {name=UseFakeFFTNoise, nodeColor=<rgba 0 1 0.3906 1>}
   Fake FFT Noise Speed X
   Fake FFT Noise Pow
   FFT_IntensityMult=LinearInterpolate(Alpha=Saturate(None=@n16)) {constA=0.15, constB=0.45}
   FFT_Add=@n17
  ) {materialFunction=/SparksCosmetics/Items/Auras/Materials/MF/MF_FX_FFT_EQ_Cosmetic}
  Add(A=@n18."FFT Sampled (Bar)") {constB=-0.1}
  n19 = TextureCoordinate()
  n20 = Constant2Vector() {g=1}
  n21 = Saturate(None=MaterialFunctionCall!(
    In TexCoords=Multiply(A=@n19, B=@n20)
    Fake FFT Noise Texture
    Fake FFT Noise Size
    Fake FFT Noise Speed X=Constant() {r=0.35}
    Fake FFT Noise Speed Y
    Fake FFT Noise Intensity Mult
    Fake FFT Noise Pow
   ) {materialFunction=/Game/Effects/Niagara/MasterMaterials/Functions/Mojo_MF/MF_FX_Fake_FFT_Noise}.Result Horizontal)
  n22 = Multiply(A=LinearInterpolate(Alpha=@n21) {constA=0.2, constB=0.15}, B=NamedRerouteUsage())
  Add(A=@n22, B=@n17)
  n23 = NamedRerouteUsage()
  n24 = NamedRerouteUsage()
  n25 = NamedRerouteUsage()
  n26 = Frac(None=Multiply(A=@n25, B=AppendVector(A=@n23, B=@n24)))
  n27 = Saturate(None=MaterialFunctionCall!(In TexCoords=@n26, Width=@n9) {materialFunction=/Game/Effects/Niagara/MasterMaterials/Functions/Mojo_MF/MF_FX_SquareEdgeMask}.Mask One Minus)
  n28 = MaterialFunctionCall!(TexCoords=@n26, Edge Sharpness=Constant() {r=1}, Edge Subtract=Constant() {r=0.5}) {materialFunction=/Game/Effects/Niagara/MasterMaterials/Functions/Mojo_MF/MF_FX_UV_EdgeMask}
  Add(A=@n27, B=NamedRerouteDeclaration(None=Multiply(A=@n28.1-Edge Mask)) {name=BoxSoftMask, nodeColor=<rgba 0 0.7656 1 1>})
  Constant() {desc=Number of bars height, r=15}
  Constant() {desc=Bar Aount Width, r=15}
  Desaturation(None=@n8, Fraction=Constant() {r=0.35})
  n29 = StaticSwitchParameter(True=ComponentMask(None=@n25) {g=true}, False=NamedRerouteUsage()) {parameterName=Use Smooth Gradient?}
  n30 = NamedRerouteUsage()
  n31 = Multiply(A=@n29, B=@n30)
  n32 = SmoothStep(Value=@n31) {constMax=0.08, constMin=0.07}
  Divide(A=@n32) {constB=0.5}
  n33 = CollectionParameter() {collection=&MPC_Sparks_Audio_Parameters, parameterName=Amplitude}
  LinearInterpolate(Alpha=@n33) {constA=0.75}
  n34 = DynamicParameter() {paramNames [Is in Front End, Param2, Param3, Param4]}
  n35 = MaterialFunctionCall(
   Square Grid Width=@n9
   RandomizeUVPlacement=@n11
   RadialUV=@n12
   BarAmount Vert=@n15
   BarAmount=@n10
   BarWidth=@n13
   MountainHardness=@n14
   FFT=NamedRerouteUsage()
   Use Fake FFT Noise=StaticBool() {value=true}
   FFT_IntensityMult=LinearInterpolate(Alpha=@n21) {constA=0.15, constB=0.45}
  ) {materialFunction=&MF_FX_FFT_EQ_Cosmetic}
  n36 = Multiply(A=@n34.Is in Front End, B=@n35."FFT Sampled (Bar)")
  LinearInterpolate(
   A=@n18."FFT Sampled (Bar)"
   B=NamedRerouteDeclaration(None=@n36) {name=FakeFFTSampledValue, nodeColor=<rgba 1 0 0.6562 1>}
   Alpha=@n34.Is in Front End
  )
  n37 = Step(Y=@n29, X=@n30)
  n38 = Multiply(A=@n37, B=NamedRerouteUsage())
  n39 = Saturate(None=Step(Y=OneMinus(None=@n38)) {constX=0.5})
  n40 = Multiply(A=NamedRerouteUsage(), B=@n39)
  n41 = Sine(None=@n32)
  n42 = OneMinus(None=Multiply(A=Saturate(None=@n41), B=NamedRerouteUsage()))
  LinearInterpolate(A=@n40, B=Constant3Vector() {constant=<rgba 1 0 0 0>}, Alpha=Saturate(None=Step(Y=@n42) {constX=0.5}))
  n43 = Constant2Vector() {g=1, r=2}
  n44 = TextureCoordinate()
  n45 = MaterialFunctionCall!(In Texcoords=Add(A=Multiply(A=@n44, B=@n43), B=Constant2Vector() {r=-0.5})) {materialFunction=/Game/Effects/Fort_Effects/MaterialFunctions/MF_Circle_SDF}
  MaterialFunctionCall!(x=@n45.Result, size=Constant() {r=0.98}) {materialFunction=/Game/Effects/Fort_Effects/MaterialFunctions/MF_Fill}
  Max(A=@n22, B=Constant() {r=0.1}) {desc=Make sure a minimum Ring of bars are displayed in the center}
  n46 = Constant() {r=1}
  Multiply(A=@n36, B=If(A=Step(X=@n33) {constX=0.001, constY=0.001}, "A > B"=Constant(), "A == B"=@n46, "A < B"=@n46)) {constB=0}
  Multiply(A=@n35."FFT Sampled (Mountain)")
  n47 = Saturate(None=@n18."FFT Sampled (Bar)")
  Multiply(
   A=@n47
   B=LinearInterpolate(Alpha=CollectionParameter() {collection=&MPC_Sparks_Audio_Parameters, parameterName=BeatLifetime}) {constA=0.55}
  )
  Multiply(A=@n18."FFT Sampled (Mountain)")
  Multiply(A=@n31) {constB=0.9}
  n48 = NamedRerouteUsage()
  n49 = ComponentMask(None=NamedRerouteUsage()) {g=true}
  n50 = Divide(A=Floor(None=Multiply(A=@n49, B=@n48)), B=@n48)
  n51 = NamedRerouteDeclaration(None=@n50) {name=SolidSteppedMask, nodeColor=<rgba 1 0 0 1>}
  n52 = NamedRerouteUsage()
  n53 = Time()
  n54 = Constant() {r=0.2}
  n55 = LinearInterpolate(A=Frac(None=Multiply(A=@n53, B=@n54)), B=Constant() {desc=manual wipe % Value}, Alpha=Constant() {desc="Use Manual Wipe? "}) {constAlpha=1}
  Multiply(A=Step(Y=Multiply(A=@n51, B=@n52), X=LinearInterpolate(Alpha=@n55) {constA=1, constB=0}), B=NamedRerouteUsage())
  n56 = NamedRerouteUsage()
  Multiply(A=NamedRerouteUsage(), B=@n56)
  n57 = NamedRerouteUsage()
  n58 = NamedRerouteUsage()
  n59 = Constant() {r=0.1}
  n60 = MaterialFunctionCall(Value_Morning=@n59, Value_Day=@n59, Value_Evening=@n59, Value_Night=Constant() {r=0.45}) {materialFunction=&Scalar_Time_of_Day}
  n61 = NamedRerouteUsage()
  n62 = ComponentMask(None=NamedRerouteUsage()) {g=true}
  n63 = Add(A=@n56) {constB=0.15}
  n64 = Multiply(A=Power(Base=SmoothStep(Min=@n62, Value=@n63) {constMax=0.9}) {constExponent=0.25}) {constB=0.155}
  n65 = Constant() {r=1}
  n66 = Constant() {r=0.01}
  n67 = ComponentMask(None=TextureCoordinate()) {r=true}
  n68 = Multiply(A=OneMinus(None=@n67), B=@n67)
  n69 = Multiply(
   A=Power(Base=MaterialFunctionCall(Edge Sharpness=@n65, Edge Subtract=@n66) {materialFunction=&MF_FX_UV_EdgeMask}.1-Edge Mask)
   B=Multiply(A=Power(Base=@n68) {constExponent=0.5}) {constB=3}
  )
  n70 = Saturate(None=Multiply(A=Multiply(A=@n64) {constB=1.5}, B=Multiply(A=@n69)))
  n71 = Constant() {r=1}
  n72 = MaterialFunctionCall(Value_Morning=@n71, Value_Day=@n71, Value_Evening=@n71, Value_Night=Constant() {r=2.65}) {materialFunction=&Scalar_Time_of_Day}
  n73 = Add(
   A=LinearInterpolate(A=Multiply(A=@n58, B=@n60.Result) {constB=0.1}, B=@n58, Alpha=@n61)
   B=NamedRerouteDeclaration(None=Multiply(A=@n70, B=@n72.Result)) {name=EdgeGlowMask, nodeColor=<rgba 1 0 0.5859 1>}
  )
  n74 = Saturate(None=@n73)
  n75 = LinearInterpolate(A=Multiply(A=Constant() {r=0.35}, B=Multiply(A=@n57, B=@n74)), B=@n74, Alpha=@n61)
  Multiply(A=@n75, B=NamedRerouteUsage())
  n76 = Power(Base=@n45.Result) {constExponent=7}
  n77 = ComponentMask(None=@n44) {g=true}
  Multiply(A=@n76, B=Power(Base=OneMinus(None=@n77)) {constExponent=4}) {desc=Gausian Curve Mask}
  n78 = NamedRerouteUsage()
  Multiply(A=@n78, B=@n37)
  Multiply(A=@n40)
  n79 = LinearInterpolate(A=@n1, B=@n8) {constAlpha=0.65}
  NamedRerouteDeclaration(None=LinearInterpolate(A=Multiply(A=@n79) {constB=2}, B=@n8, Alpha=NamedRerouteUsage())) {name=Color Gradient, nodeColor=<rgba 1 0.2109 0 1>}
  NamedRerouteDeclaration(None=@n39) {name=Bar Stepped Mask, nodeColor=<rgba 1 0 0.4687 1>}
  NamedRerouteDeclaration(None=@n15) {name=BarAmountVert, nodeColor=<rgba 1 0.5391 0 1>}
  NamedRerouteDeclaration(None=@n18.EQ Bar Boxes) {name=EQBarBoxes, nodeColor=<rgba 0.1953 1 0 1>}
  n80 = TextureCoordinate()
  n81 = ComponentMask(None=@n80) {g=true}
  n82 = MaterialFunctionCall(Float2=AppendVector(A=ComponentMask(None=@n80) {r=true}, B=OneMinus(None=@n81))) {materialFunction=&BreakOutFloat2Components}
  n83 = Step(Y=@n82.G) {constX=0.5}
  n84 = OneMinus(None=@n83)
  NamedRerouteDeclaration(None=@n84) {name=TopHalfMask, nodeColor=<rgba 0 1 0.625 1>}
  n85 = TextureCoordinate()
  n86 = Add(A=Multiply(A=@n85, B=Constant2Vector() {g=0.97, r=0.55}), B=Constant2Vector() {g=-0.1235, r=-0.3765})
  n87 = TextureObject() {samplerType=SAMPLERTYPE_Masks, texture=/Game/Global/Textures/Default/Noise/T_Generic_Perlin}
  n88 = Add(A=Multiply(A=@n85, B=Constant2Vector() {g=1.85, r=1.25}), B=Constant2Vector() {g=0.5, r=0.1})
  n89 = Multiply(
   A=TextureSample(UVs=Panner(Coordinate=@n86) {bFractionalPart=true, speedX=0.1, speedY=0.05}, Tex=@n87).R
   B=TextureSample(UVs=Panner(Coordinate=@n88) {bFractionalPart=true, speedX=0.075, speedY=-0.065}, Tex=@n87).G
  )
  NamedRerouteDeclaration(None=Saturate(None=Add(A=@n89) {constB=0.1})) {name=Noise mask, nodeColor=<rgba 0.4531 0 1 1>}
  NamedRerouteDeclaration(None=@n35."FFT Sampled (Mountain)") {name=MountainGradient, nodeColor=<rgba 0 1 0.3203 1>}
  n90 = ComponentMask(None=NamedRerouteUsage()) {r=true}
  n91 = NamedRerouteDeclaration(None=@n10) {name=BarAmount, nodeColor=<rgba 1 0.5391 0 1>}
  n92 = Sine(None=Add(A=Multiply(A=@n90, B=@n91) {constB=10}) {constB=0.2})
  n93 = Constant() {desc=Bar Width, r=0.5}
  NamedRerouteDeclaration(None=Step(Y=ConstantBiasScale(None=@n92), X=NamedRerouteDeclaration(None=@n93) {name=BarWidth, nodeColor=<rgba 1 0 0.2812 1>}) {constX=0.1}) {name=BarWidthSine, nodeColor=<rgba 0.9219 1 0 1>}
  NamedRerouteDeclaration(None=LinearInterpolate(Alpha=Saturate(None=@n76))) {name=MiddleMask, nodeColor=<rgba 0 1 0.4375 1>}
  n94 = NamedRerouteDeclaration(None=TextureObject() {texture=/FMAudioReactiveData/AudioAnalysis/T_Sparks_Audio_FFT}) {name=FFTTexObject, nodeColor=<rgba 0 1 0.2734 1>}
  NamedRerouteDeclaration(None=Saturate(None=TextureSample(Tex=@n94) {texture=/Game/Effects/Fort_Effects/Textures/Fire/T_Fire_Mask}.R)) {name=FFT_Sample, nodeColor=<rgba 0.125 0 1 1>}
  n95 = ObjectOrientation()
  n96 = CameraPositionWS()
  n97 = Constant3Vector() {constant=<rgba 1 1 0 0>}
  MaterialExpressionNormalize_0 = Normalize(VectorInput=Subtract(A=ObjectPositionWS(), B=Multiply(A=@n96, B=@n97)))
  n98 = Custom!(In=@MaterialExpressionNormalize_0) {
   code="float2 output; float2 Input1; output = atan2 (In.y,In.x); return (output);"
   description=Billboard
   outputType=CMOT_Float1
   inputs [{inputName=In, input {expression=&self:MaterialExpressionNormalize_0}}]
  }
  n99 = Constant() {r=3.142}
  n100 = Add(A=Divide(A=Divide(A=@n98, B=@n99), B=Constant() {r=2}), B=Constant() {r=0.25})
  n101 = AppendVector(A=TextureCoordinate(), B=ComponentMask(None=TextureCoordinate()) {r=true})
  NamedRerouteDeclaration(None=RotateAboutAxis(
    NormalizedRotationAxis=@n95
    RotationAngle=@n100
    PivotPoint=TransformPosition(None=@n101) {transformType=TRANSFORMPOSSOURCE_World}
    Position=WorldPosition().XYZ
   )) {name=WPO Result, nodeColor=<rgba 0.4062 1 0 1>}
  NamedRerouteDeclaration(None=MaterialFunctionCall(Rotation Axis=@n95, Rotation Angle=@n100) {materialFunction=&FixRotateAboutAxisNormals}.New World Space Vertex Normal) {name=Normal, nodeColor=<rgba 0.2422 0 1 1>}
  n102 = MaterialFunctionCall(X=@n82.R, Y=Add(A=OneMinus(None=@n82.G)) {constB=-0.5}) {materialFunction=&MakeFloat2}
  n103 = MaterialFunctionCall(X=@n82.R, Y=Add(A=@n82.G) {constB=-0.5}) {materialFunction=&MakeFloat2}
  NamedRerouteDeclaration(None=Add(A=Multiply(A=@n102.Result, B=@n83), B=Multiply(A=@n103.Result, B=@n84))) {name=UVs, nodeColor=<rgba 0.9688 0 1 1>}
  n104 = Subtract(A=@n49) {constB=0}
  n105 = Divide(A=Floor(None=Multiply(A=@n104, B=@n48)), B=@n48)
  NamedRerouteDeclaration(None=@n105) {name=SteppedVertGradient, nodeColor=<rgba 1 0 0.7031 1>}
  NamedRerouteDeclaration(None=@n27) {name=BoxMask, nodeColor=<rgba 0 1 0.1094 1>}
  NamedRerouteDeclaration(None=@n47) {name=FFTSampledBar, nodeColor=<rgba 1 0.4219 0 1>}
  NamedRerouteUsage()
  NamedRerouteUsage()
  NamedRerouteUsage()
  NamedRerouteUsage()
  NamedRerouteUsage()
  NamedRerouteUsage()
  OneMinus(None=@n50)
  OneMinus(None=@n78)
  OneMinus(None=@n105)
  n106 = Constant() {desc=Boost, r=8}
  n107 = Constant() {r=10}
  n108 = Constant() {r=3}
  n109 = NamedRerouteUsage()
  Reroute(None=Multiply(
    A=MaterialFunctionCall(Value_Morning=@n106, Value_Day=@n107, Value_Evening=@n106, Value_Night=@n108) {materialFunction=&Scalar_Time_of_Day}.Result
    B=Multiply(A=@n109, B=@n73)
   ))
  Reroute(None=@n75)
  TextureCoordinate()
 ]
}
